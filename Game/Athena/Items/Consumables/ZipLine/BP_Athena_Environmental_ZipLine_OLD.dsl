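; BP_Athena_Environmental_ZipLine_OLD is an environmental zipline Blueprint, named as an older version and stored with Athena consumables.
; It overrides GetMinDistanceFromInteraction to return a fixed zero distance.
; Its construction script only calls the parent, and no other logic appears in its graphs.
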(fn GetMinDistanceFromInteraction
  (return 0.0))

(fn ConstructionScript
  (|CallParentFunction))
