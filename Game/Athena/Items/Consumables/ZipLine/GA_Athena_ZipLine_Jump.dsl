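; GA_Athena_ZipLine_Jump is a gameplay ability that launches the player pawn in a jump, by its name off a zipline.
; On activation it scales the pawn's velocity by a lateral damper, caps the XY part at a maximum lateral speed, and sets Z to a jump strength.
; LaunchCharacterJump applies this impulse, with one flag depending on whether the pawn is moving upward.
; All three tuning values come from ScalableFloats.
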(event Ability|EventActivateAbility
  (bind _maxlateralspeed (Variables|Default|GetMaxLateralSpeed))
  (bind _returnvalue (ScalableFloat|GetValueatLevel _maxlateralspeed))
  (bind _jumpvertstrength (Variables|Default|GetJumpVertStrength))
  (bind _returnvalue_2 (ScalableFloat|GetValueatLevel _jumpvertstrength))
  (|CallParentFunction)
  (bind _asfort_player_pawn (Utilities|Casting|CastToFortPlayerPawn (Ability|GetAvatarActorfromActorInfo)))
  (bind _output_get (Variables|Default|SetPlayerPawn _asfort_player_pawn))
  (bind (_returnvalue_x _returnvalue_y) (Math|Vector|BreakVector (* (Transformation|GetVelocity _output_get) (ScalableFloat|GetValueatLevel (Variables|Default|GetJumpLateralDamper)))))
  (bind _output_get_2 (Variables|SetImpulseVec_X (Math|Vector|MakeVector _returnvalue_x _returnvalue_y _returnvalue_2)))
  (bind (_output_get_x _output_get_y) (Math|Vector|BreakVector _output_get_2))
  (if (> (Math|Vector|VectorLengthSquared (Math|Vector|MakeVector _output_get_x _output_get_y)) (* _returnvalue _returnvalue))
    (bind (_returnvalue_x_2 _returnvalue_y_2) (Math|Vector|BreakVector (* (Math|Vector|Normalize (Math|Vector|MakeVector _output_get_x _output_get_y)) _returnvalue)))
    (Variables|SetImpulseVec_X (Math|Vector|MakeVector _returnvalue_x_2 _returnvalue_y_2 _returnvalue_2)))
  (bind (_returnvalue_x_3 _returnvalue_y_3 _returnvalue_z) (Math|Vector|BreakVector (Transformation|GetVelocity _output_get)))
  (Pawn|Character|LaunchCharacterJump _output_get (Variables|Default|GetImpulseVec) true (> _returnvalue_z 0.0) false true))

(event Ability|EventOnEndAbility
  (params (bWasCancelled :type Boolean))
  (|CallParentFunction bWasCancelled))
