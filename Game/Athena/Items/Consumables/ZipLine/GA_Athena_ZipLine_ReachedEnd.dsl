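; Zipline gameplay ability that launches the player pawn with a jump when it reaches the end of a zipline.
; Lateral velocity is damped and capped, while vertical strength follows a squared inverse of normalized speed with a minimum, so slower riders pop higher.
; The launch overrides lateral velocity, and vertical only when descending.
; Tuning comes from scalable floats, with an optional debug print.
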(event Ability|EventActivateAbility
  (bind _maxziplinevel (Variables|Default|GetMaxZiplineVel))
  (bind _returnvalue (ScalableFloat|GetValueatLevel _maxziplinevel))
  (bind _minjumpvertstrength (Variables|Default|GetMinJumpVertStrength))
  (bind _returnvalue_2 (ScalableFloat|GetValueatLevel _minjumpvertstrength))
  (bind _maxlateralspeed (Variables|Default|GetMaxLateralSpeed))
  (bind _returnvalue_3 (ScalableFloat|GetValueatLevel _maxlateralspeed))
  (bind _jumpvertstrength (Variables|Default|GetJumpVertStrength))
  (bind _returnvalue_4 (ScalableFloat|GetValueatLevel _jumpvertstrength))
  (bind _lateraldamping (Variables|Default|GetLateralDamping))
  (bind _returnvalue_5 (ScalableFloat|GetValueatLevel _lateraldamping))
  (|CallParentFunction)
  (bind _asfort_player_pawn (Utilities|Casting|CastToFortPlayerPawn (Ability|GetAvatarActorfromActorInfo)))
  (bind _output_get (Variables|Default|SetPlayerPawn _asfort_player_pawn))
  (if (Variables|Default|Getdebug)
    (Utilities|FlowControl|Sequence
      (:then_0
        (Development|PrintString (Utilities|String|Append "reached end.  Vel:" (Utilities|String|ToString(Float) (Math|Vector|VectorLength (Transformation|GetVelocity _output_get))) "    vert strength mult" (Utilities|String|ToString(Float) (* (- (Math|Float|Clamp(Float) (/ (Math|Vector|VectorLength (Transformation|GetVelocity _output_get)) _returnvalue)) 1.0) (- (Math|Float|Clamp(Float) (/ (Math|Vector|VectorLength (Transformation|GetVelocity _output_get)) _returnvalue)) 1.0))) "   speed:" (Utilities|String|ToString(Float) (Zipline|GetZiplineSpeedFactor _output_get))) (Variables|Default|Getdebug) (Variables|Default|Getdebug) "(R=1.000000,G=0.000000,B=0.028356,A=1.000000)" 2e+01))
      (:then_1
        _)))
  (bind (_returnvalue_x _returnvalue_y _returnvalue_z) (Math|Vector|BreakVector (Transformation|GetVelocity _output_get)))
  (bind _output_get_2 (Variables|Default|SetImpulseVec (* (Math|Vector|MakeVector _returnvalue_x _returnvalue_y) _returnvalue_5)))
  (bind (_output_get_x _output_get_y) (Math|Vector|BreakVector _output_get_2))
  (Variables|SetImpulseVec_X (Math|Vector|MakeVector _output_get_x _output_get_y (Math|Float|Max(Float) (* (* (- (Math|Float|Clamp(Float) (/ (Math|Vector|VectorLength (Transformation|GetVelocity _output_get)) _returnvalue)) 1.0) (- (Math|Float|Clamp(Float) (/ (Math|Vector|VectorLength (Transformation|GetVelocity _output_get)) _returnvalue)) 1.0)) _returnvalue_4) _returnvalue_2)))
  (if (> (Math|Vector|VectorLengthSquared (* (Math|Vector|MakeVector _returnvalue_x _returnvalue_y) _returnvalue_5)) (* _returnvalue_3 _returnvalue_3))
    (bind (_returnvalue_x_2 _returnvalue_y_2) (Math|Vector|BreakVector (* (Math|Vector|Normalize (* (Math|Vector|MakeVector _returnvalue_x _returnvalue_y) _returnvalue_5)) _returnvalue_3)))
    (Variables|SetImpulseVec_X (Math|Vector|MakeVector _returnvalue_x_2 _returnvalue_y_2 (Math|Float|Max(Float) (* (* (- (Math|Float|Clamp(Float) (/ (Math|Vector|VectorLength (Transformation|GetVelocity _output_get)) _returnvalue)) 1.0) (- (Math|Float|Clamp(Float) (/ (Math|Vector|VectorLength (Transformation|GetVelocity _output_get)) _returnvalue)) 1.0)) _returnvalue_4) _returnvalue_2))))
  (Pawn|Character|LaunchCharacterJump _output_get (Variables|Default|GetImpulseVec) true (< _returnvalue_z 0.0) false true))

(event Ability|EventOnEndAbility
  (params (bWasCancelled :type Boolean))
  (|CallParentFunction bWasCancelled))
