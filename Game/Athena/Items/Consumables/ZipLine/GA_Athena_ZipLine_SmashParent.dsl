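; Parent Athena zipline ability that runs a player's ride and smashes damageable structures in the way.
; The enter event grants fall damage immunity and starts ziplining, then delayed timers spawn a server-side harness and repeat async overlap smash, harness and vehicle checks.
; Exit, water, skydiving, aircraft, landing and vehicle events variously end the ride or remove immunity, sometimes delayed.
; Clients send spline zipline exit events themselves, as splines lack overlap capsules.
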(event Custom|EnterWaterExit
  (params (FortPlayerPawn :type Fort%20Player%20Pawn%20Object%20Reference))
  ;[1 1 1] Bind events and wait for tag adds
  ;;[1 1 1] Send the exit zipline event if a player enters water while ziplining.
  (Ability|SendGameplayEvent (Variables|Default|GetZiplineExitTag))
  (CallFunction|RemoveFallDamageImmunityDelayed))

(event Custom|HandleClientSplineZiplineStateChanged
  (params (bIsZiplining :type Boolean) (Pawn :type Fort%20Player%20Pawn%20Object%20Reference))
  ;[0 0.4852 0.04092] Client Handle Spline Zipline State Changed
  (if (not bIsZiplining)
    (EventDispatchers|UnbindEventfromZiplineStateChanged (Variables|Default|GetPlayerPawn) (EventDispatchers|CreateEvent :!event Custom|HandleClientSplineZiplineStateChanged))
    (Utilities|Time|SetTimerforNextTickbyEvent (delegate Custom|SplineZiplineMovementModeDelayOver))))

(event Custom|BindEvent
  (bind _outputdelegate (EventDispatchers|CreateEvent :!event Custom|OnEnterVehicle_Bind))
  (bind _playerpawn (Variables|Default|GetPlayerPawn))
  ;[1 1 1] Bind events and wait for tag adds
  (EventDispatchers|BindEventtoOnBeginSkydivingEvent _playerpawn (EventDispatchers|CreateEvent :!event Custom|OnBeginSkydiving_Bind))
  (EventDispatchers|BindEventtoOnEnteredAircraft _playerpawn (EventDispatchers|CreateEvent :!event Custom|OnEnteredAircraft_Bind))
  (EventDispatchers|BindEventtoOnPawnLanded _playerpawn (EventDispatchers|CreateEvent :!event Custom|OnPawnLanded_Bind))
  (EventDispatchers|BindEventtoOnDied _playerpawn (EventDispatchers|CreateEvent :!event Custom|OnPawnDied))
  (EventDispatchers|BindEventtoOnPlayerStartDBNO _playerpawn (EventDispatchers|CreateEvent :!event Custom|OnPawnDBNO))
  (EventDispatchers|AssignZiplinePostBegin _playerpawn (EventDispatchers|CreateEvent :!event Custom|ZiplinePostBegin_Event))
  ;;[1 1 1] Send the exit zipline event if a player enters water while ziplining.
  (EventDispatchers|BindEventtoOnPawnEnterWater _playerpawn (delegate Custom|EnterWaterExit))
  ;;[1 1 1] Check for enter water while ziplining (do not use WaitGameplayTagAdded as that can trigger immediately and we don't want that to happen at this point in initialisation when ziplining hasn't started yet. This means that we can trigger zipline & ascender when swimming but if we enter swimming later or get EnterWater event, ziplining will still end early.)
  (bind (_asynctaskproxy _tagcount) (Ability|Tasks|WaitGameplayTagCountChange (Variables|Default|GetT_Swimming))
    (:then
      (bind _asfort_player_controller (Utilities|Casting|CastToFortPlayerController (Pawn|GetController _playerpawn))
        (:then
          (Vehicle|BindEventtoOnEnterVehicleDriver _asfort_player_controller _outputdelegate)
          (Vehicle|BindEventtoOnEnterVehiclePassenger _asfort_player_controller _outputdelegate)
          ;;[0 0.4852 0.04092] Spline Ziplines need exit event to happen on client in order
          ;; for this ability to detect gameplay event correct remove immunity GEs (based on net execution policy LocalPredicted)
          ;;
          ;;
          (Networking|CanExecuteCosmeticEvents
            (:True
              (EventDispatchers|BindEventtoZiplineStateChanged _playerpawn (EventDispatchers|CreateEvent :!event Custom|HandleClientSplineZiplineStateChanged)))
            (:False
              _)) ; is client?
        )
        (:CastFailed
          _)))
    (:TagCountChanged
      (if (> _tagcount 0)
        (CallFunction|SwimmingStarted)))))

(event Custom|SplineZiplineMovementModeDelayOver
  ;[0 0.4852 0.04092] Client Handle Spline Zipline State Changed
  (Ability|SendGameplayEventtoActor (Variables|Default|GetPlayerPawn) (Variables|Default|GetZiplineExitTag) (Utilities|Struct|MakeGameplayEventData (Variables|Default|GetZiplineExitTag))) ; will cause this ability to go through teardown state,\nwe need to do the event incase anything else decides to \nlisten to the event. spine ziplines don't have overlap \ncapsule component like linear ziplines have.
)

(event Custom|GameplayTagEvent_ZiplineEnter
  (params (EventData :type Gameplay%20Event%20Data%20Structure%20%28by%20ref%29))
  (bind (_eventtag _instigator _target _optionalobject _optionalobject2) (Utilities|Struct|BreakGameplayEventData EventData))
  (bind _playerpawn (Variables|Default|GetPlayerPawn))
  (bind _returnvalue (Ability|GetAbilitySystemComponent _playerpawn))
  ;[1 1 1] EVENT -> ActivateAbilityFromEvent
  (if (not (GameplayTags|HasAnyMatchingGameplayTags _returnvalue (Variables|Default|GetZiplineEnterBlockedTags)))
    (GameplayEffects|ApplyGameplayEffectToTarget :self _returnvalue :GameplayEffectClass (Variables|Default|GetGE%20Ziplining%20Passive) :Target _returnvalue :Level 0.0) ; blocks sword abilities
    (bind _returnvalue_2 (GameplayEffects|ApplyGameplayEffectToTarget :self _returnvalue :GameplayEffectClass "/Game/Athena/Items/Consumables/ZipLine/GE_Athena_ZipLine_FallDamageImmunityFX.GE_Athena_ZipLine_FallDamageImmunityFX_C" :Target _returnvalue :Level 0.0)) ; Fall Damage ImmunityFX
    (Variables|Default|SetFallImmunityGEFX _returnvalue_2)
    (Variables|/Script/FortniteGame.FortPawn|SetbIgnoreNextFallingDamage true (Variables|Default|GetPlayerPawn))
    ;;[0 0 0] Get zipline that player is overlapping
    (bind _asfort_athena_zipline (Utilities|Casting|CastToFortAthenaZipline (Utilities|Struct|BreakGameplayEventData EventData)))
    (Variables|Default|SetZipline _asfort_athena_zipline)
    ;;[0 0 0] Get zipline Interact Component that player is overlapping
    (bind _asprimitive_component (Utilities|Casting|CastToPrimitiveComponent _optionalobject2)
      (:then
        _)
      (:CastFailed
        _))
    (Variables|Default|SetZiplineInteractComponent _asprimitive_component)
    (CallFunction|BindEvent)
    (CallFunction|GameplayEventZiplineEnterBeginZiplining)
    (if (and (Networking|HasAuthority (Variables|Default|GetPlayerPawn)) (CallFunction|IsZiplining))
      (CallFunction|ZiplinePostBeginEvent))))

(event Custom|Pre_BeginZiplinging
  ;[1 1 1] EVENT -> ActivateAbilityFromEvent
  (CallFunction|BreakNearbyStructures (Variables|Default|GetPlayerPawn)))

(event Custom|HandleGameplayEventTag
  (params (EventData :type Gameplay%20Event%20Data%20Structure%20%28by%20ref%29))
  ;[1 1 1] EVENT -> ActivateAbilityFromEvent
  (switch Utilities|FlowControl|Switch|SwitchonGameplayTag (Utilities|Struct|BreakGameplayEventData EventData)
    (:Abilities.Generic.Athena.ZipLine.Enter
      (CallFunction|GameplayTagEventZiplineEnter EventData))
    (:Abilities.Generic.Athena.ZipLine.Exit
      (CallFunction|GameplayTagEventZiplineExit EventData))))

(event Ability|EventActivateAbilityFromEvent
  (params (EventData :type Gameplay%20Event%20Data%20Structure%20%28by%20ref%29))
  ;[1 1 1] EVENT -> ActivateAbilityFromEvent
  (|CallParentFunction EventData)
  (bind _asfort_player_pawn (Utilities|Casting|CastToFortPlayerPawn (Ability|GetAvatarActorfromActorInfo)))
  (Variables|Default|SetPlayerPawn _asfort_player_pawn)
  (CallFunction|HandleGameplayEventTag EventData))

(event Custom|GameplayEvent_ZiplineEnter_BeginZiplining
  ;[1 1 1] EVENT -> ActivateAbilityFromEvent
  (CallFunction|PreBeginZiplinging)
  (Zipline|BeginZiplining (Variables|Default|GetPlayerPawn) (Variables|Default|GetZipline) (Variables|/Script/Engine.Character|GetMesh (Variables|Default|GetPlayerPawn)) (Variables|Default|GetZiplineInteractComponent) "Unset")
  (Ability|ExecuteGameplayCueOnOwner (Variables|Default|GetAttachVFXCue)))

(event Custom|OnEnteredAircraft_Bind
  ;[1 1 1] EVENT -> ActivateAbilityFromEvent
  (CallFunction|ZipliningComplete)
  (CallFunction|RemoveFallDamageImmunity))

(event Custom|OnBeginSkydiving_Bind
  ;[1 1 1] EVENT -> ActivateAbilityFromEvent
  (CallFunction|ZipliningComplete)
  (CallFunction|RemoveFallDamageImmunity))

(event Custom|SwimmingStarted
  ;[1 1 1] EVENT -> ActivateAbilityFromEvent
  (CallFunction|ZipliningComplete)
  (CallFunction|RemoveFallDamageImmunity))

(event Custom|GameplayTagEvent_ZiplineExit
  (params (EventData :type Gameplay%20Event%20Data%20Structure%20%28by%20ref%29))
  ;[1 1 1] EVENT -> ActivateAbilityFromEvent
  (CallFunction|ZipliningComplete))

(event Custom|OnPawnDied
  (params (DamagedActor :type Actor%20Object%20Reference) (Damage :type Float%20%28single-precision%29) (InstigatedBy :type Controller%20Object%20Reference) (DamageCauser :type Actor%20Object%20Reference) (HitLocation :type Vector) (FHitComponent :type Primitive%20Component%20Object%20Reference) (BoneName :type Name) (Momentum :type Vector))
  ;[1 1 1] EVENT -> ActivateAbilityFromEvent
  ;;[1 1 1] OnDied only called on server, so end ability to ensure the client gets their ability ended too
  (Ability|CancelAbility))

(event Custom|ZiplinePostBegin_Event
  ;[1 1 1] EVENT -> ActivateAbilityFromEvent
  (Utilities|FlowControl|Sequence
    (:then_0
      ;;[1 1 1] QUEST STUFF
      (CallFunction|CheckforQuestProgress))
    (:then_1
      ;;[1 1 1] Mitigates race condition issues where player is being snapped to a location at the same time the motor is being attached and snapped to the player.
      (Utilities|Time|SetTimerbyEvent (EventDispatchers|CreateEvent :!event Custom|DelaySpawnMotor) 0.1)
      (Utilities|Time|SetTimerbyEvent (EventDispatchers|CreateEvent :!event Custom|StartSurroundingsCheck) 0.1))
    (:then_2
      ;;[1 1 1] Trigger Entered Anim Input Event
      (bind _returnvalue (Class|FortAnimInputEvent|CreateAnimInputEvent "aie.Player.Zipline.Entered"))
      (Animation|TriggerAnimInputEvent (Variables|Default|GetPlayerPawn) _returnvalue))))

(event Custom|ZipliningComplete
  (bind _ziplineharness (Variables|Default|GetZiplineHarness))
  ;[1 1 1] EVENT -> ActivateAbilityFromEvent
  (Ability|RemoveGameplayEffectFromOwnerWithGrantedTags "(GameplayTags=((TagName=\"Abilities.Generic.Athena.ZipLine\")))") ; remove zipline tag
  (Zipline|EndZiplining (Variables|Default|GetPlayerPawn) false false)
  (Utilities|Time|ClearandInvalidateTimerbyHandle (Variables|Default|GetSurroundingsCheckTimerHandle))
  (EventDispatchers|BindEventtoZiplineStateChanged (Variables|Default|GetPlayerPawn) (EventDispatchers|CreateEvent :!event Custom|HandleZiplineStateChanged))
  (Utilities|IsValid _ziplineharness
    (:Is%20Valid
      (Actor|DestroyActor _ziplineharness))
    (:Is%20Not%20Valid
      _))
  ;;[1 1 1] Trigger Exited Anim Input Event
  (bind _returnvalue (Class|FortAnimInputEvent|CreateAnimInputEvent "aie.Player.Zipline.Exited"))
  (Animation|TriggerAnimInputEvent (Variables|Default|GetPlayerPawn) _returnvalue))

(event Custom|FallingTagRemoved
  ;[1 1 1] EVENT -> ActivateAbilityFromEvent
  ;;[0 1 0.879] Retro compatibility - used by children
  (if (not (GameplayTags|HasAnyMatchingGameplayTags (Variables|Default|GetPlayerPawn) (Variables|Default|GetTagsThatBlockFallImmunityRemoval)))
    (CallFunction|RemoveFallDamageImmunityDelayed)))

(event Custom|OnPawnDBNO)

(event Custom|HandleZiplineStateChanged
  (params (bIsZiplining :type Boolean) (Pawn :type Fort%20Player%20Pawn%20Object%20Reference))
  (bind _playerpawn (Variables|Default|GetPlayerPawn))
  ;[1 1 1] CUSTOM EVENT -> HandleZiplineStateChanged
  (if (not bIsZiplining)
    (Utilities|FlowControl|Sequence
      (:then_0
        (if (not (CallFunction|IsZiplining))
          (Utilities|Time|ClearandInvalidateTimerbyHandle (Variables|Default|GetSurroundingsCheckTimerHandle))
          (EventDispatchers|UnbindEventfromZiplineStateChanged _playerpawn (EventDispatchers|CreateEvent :!event Custom|HandleZiplineStateChanged))
          (EventDispatchers|UnbindEventfromZiplinePostBegin _playerpawn (EventDispatchers|CreateEvent :!event Custom|ZiplinePostBegin_Event))))
      (:then_1
        ;;[0 0 0] Destroy motor when ending zip lining
        (bind _ziplineharness (Variables|Default|GetZiplineHarness)
          (:then
            (Actor|DestroyActor _ziplineharness))
          (:else))))))

(event Custom|StartSurroundingsCheck
  ;[1 1 1] Start Surroundings Check
  (bind _returnvalue (Utilities|Time|SetTimerbyEvent (EventDispatchers|CreateEvent :!event Custom|SurroundingsCheck) (Variables|Default|GetUpdateIntervalForSmashVolume) true))
  (Variables|Default|SetSurroundingsCheckTimerHandle _returnvalue))

(event Custom|DelaySpawnMotor
  ;[1 1 1] CUSTOM EVENT -> DelaySpawnMotor
  (if (CallFunction|IsZiplining)
    (CallFunction|SpawnMotor (Variables|Default|GetPlayerPawn))))

(event Custom|SurroundingsCheck
  ;[1 1 1] SurroundingsCheck
  (Utilities|FlowControl|Sequence
    (:then_0
      (CallFunction|SurroundingsCheckSmashObjects))
    (:then_1
      (CallFunction|SurroundingsCheckMotorCheck))
    (:then_2
      (CallFunction|SurruondingsCheckVehicleCheck))))

(event Custom|SurruondingsCheck_VehicleCheck
  (bind _playerpawn (Variables|Default|GetPlayerPawn))
  (bind _ziplineharness (Variables|Default|GetZiplineHarness))
  ;[1 1 1] SurroundingsCheck
  (if (Vehicle|IsinInteractBlockingVehicle (Variables|Default|GetPlayerPawn))
    (Zipline|EndZiplining _playerpawn false false)
    (Pawn|Components|CharacterMovement|SetMovementMode (Variables|/Script/Engine.Character|GetCharacterMovement _playerpawn) "MOVE_Custom" (Variables|Default|GetDrivingCustomMovementMode))
    (Utilities|Time|ClearandInvalidateTimerbyHandle (Variables|Default|GetSurroundingsCheckTimerHandle))
    (EventDispatchers|BindEventtoZiplineStateChanged (Variables|Default|GetPlayerPawn) (EventDispatchers|CreateEvent :!event Custom|HandleZiplineStateChanged))
    (Utilities|IsValid _ziplineharness
      (:Is%20Valid
        (Actor|DestroyActor _ziplineharness)))))

(event EventEndAsyncBreakNearbyStructuresOverlap
  (params (OverlapActors :type Array%20of%20Actor%20Object%20References))
  (bind _ziplineharness (Variables|Default|GetZiplineHarness))
  (bind _ziplineharness_2 (Variables|Default|GetZiplineHarness))
  ;[1 1 1] Async Break structures in the way
  (for _array_element OverlapActors
    ;;[0 0 0] Check for player and set different destroy VFX
    (Utilities|Casting|CastToFortPlayerPawn _array_element
      (:then
        (Utilities|IsValid _ziplineharness_2
          (:Is%20Valid
            (Variables|/Game/Gadgets/Zipline/Meshes/BP_ZipLine_Athena_Harness.BP_ZipLine_Athena_Harness_C|SetHasHitAnotherPlayer true _ziplineharness_2))))
      (:CastFailed
        (bind _asbuilding_actor (Utilities|Casting|CastToBuildingActor _array_element)
          (:then
            (Utilities|IsValid (Actor|GetComponentbyClass _asbuilding_actor "/Script/FortniteGame.FortHeldObjectComponent")
              (:Is%20Not%20Valid
                (Utilities|Casting|CastToFortAthenaSupplyDropBalloon _asbuilding_actor
                  (:then
                    _)
                  (:CastFailed
                    (if (and (Variables|/Script/Engine.Actor|GetbCanBeDamaged _asbuilding_actor) (not (GameplayTags|HasMatchingGameplayTag _asbuilding_actor "(TagName=\"Abilities.Generic.Athena.ZipLine.ZiplineSmashImmune\")")))
                      (Ability|ApplyGameplayEffectToTarget (Ability|TargetData|AbilityTargetDatafromActor _asbuilding_actor) (Variables|Default|GetStructureDamageGE))
                      (Ability|ExecuteGameplayCueOnOwner (Variables|Default|GetSmashCue)))))
                (Utilities|IsValid _ziplineharness
                  (:Is%20Valid
                    (Variables|/Game/Gadgets/Zipline/Meshes/BP_ZipLine_Athena_Harness.BP_ZipLine_Athena_Harness_C|SetHasHitAnotherPlayer false _ziplineharness))))))
          (:CastFailed
            _))))))

(event Custom|Remove%20Fall%20Damage%20Immunity%20Delayed
  ;[1 1 1] Delayed Fall Immunity Removal
  (if (not (Utilities|Time|IsValidTimerHandle (Variables|Default|GetTH_DelayFallDamageRemoval)))
    (bind _returnvalue (Utilities|Time|SetTimerbyEvent (EventDispatchers|CreateEvent :!event Custom|RemoveFallDamageImmunity) 0.1))
    (Variables|Default|SetTH_DelayFallDamageRemoval _returnvalue)))

(event Custom|SurroundingsCheck_MotorCheck
  ;[1 1 1] SurroundingsCheck
  (Utilities|IsValid (Variables|Default|GetZiplineHarness)
    (:Is%20Not%20Valid
      (CallFunction|SpawnMotor (Variables|Default|GetPlayerPawn)))))

(event Custom|SurroundingsCheck_SmashObjects
  ;[1 1 1] SurroundingsCheck
  ;;[0 0 0] Smash nearby structures
  (CallFunction|BreakNearbyStructures (Variables|Default|GetPlayerPawn)))

(event Custom|RemoveFallDamageImmunity
  (bind _outputdelegate (EventDispatchers|CreateEvent :!event Custom|OnEnterVehicle_Bind))
  (bind _playerpawn (Variables|Default|GetPlayerPawn))
  ;[1 1 1] CUSTOM EVENT -> RemoveFallDamageImmunity
  (Utilities|Time|ClearandInvalidateTimerbyHandle (Variables|Default|GetTH_DelayFallDamageRemoval))
  (Ability|RemoveGameplayEffectFromOwnerWithHandle (Variables|Default|GetFallImmunityGEFX))
  (Variables|/Script/FortniteGame.FortPawn|SetbIgnoreNextFallingDamage false _playerpawn)
  ;[1 1 1] Unbind Events
  (EventDispatchers|UnbindEventfromOnBeginSkydivingEvent _playerpawn (EventDispatchers|CreateEvent :!event Custom|OnBeginSkydiving_Bind))
  (EventDispatchers|UnbindEventfromOnEnteredAircraft _playerpawn (EventDispatchers|CreateEvent :!event Custom|OnEnteredAircraft_Bind))
  (EventDispatchers|UnbindEventfromOnDied _playerpawn (EventDispatchers|CreateEvent :!event Custom|OnPawnDied))
  (EventDispatchers|UnbindEventfromOnPlayerStartDBNO _playerpawn (EventDispatchers|CreateEvent :!event Custom|OnPawnDBNO))
  (EventDispatchers|UnbindEventfromZiplinePostBegin _playerpawn (EventDispatchers|CreateEvent :!event Custom|ZiplinePostBegin_Event))
  (EventDispatchers|UnbindEventfromOnPawnLanded _playerpawn (EventDispatchers|CreateEvent :!event Custom|OnPawnLanded_Bind))
  (bind _asfort_player_controller (Utilities|Casting|CastToFortPlayerController (Pawn|GetController _playerpawn)))
  (Vehicle|UnbindEventfromOnEnterVehicleDriver _asfort_player_controller _outputdelegate)
  (Vehicle|UnbindEventfromOnEnterVehiclePassenger _asfort_player_controller _outputdelegate))

(event Custom|OnPawnLanded_Bind
  (params (Hit :type Hit%20Result%20Structure%20%28by%20ref%29))
  ;[1 1 1] CUSTOM EVENT -> OnPawnLanded_Bind
  (if (not (Zipline|WantstoGofromVehicletoZipline (Variables|Default|GetPlayerPawn) :!class /Script/FortniteGame.FortPlayerPawn))
    (CallFunction|RemoveFallDamageImmunityDelayed)))

(event Custom|OnEnterVehicle_Bind
  ;[1 1 1] CUSTOM EVENT -> OnEnterVehicle_Bind
  (CallFunction|RemoveFallDamageImmunity))

(event Ability|EventOnEndAbility
  (params (bWasCancelled :type Boolean))
  (bind _ziplineharness (Variables|Default|GetZiplineHarness))
  ;[1 1 1] End Ability - Cleanup harness/etc if needed
  (if bWasCancelled
    (CallFunction|RemoveFallDamageImmunity))
  (Utilities|IsValid _ziplineharness
    (:Is%20Valid
      (Actor|DestroyActor _ziplineharness))
    (:Is%20Not%20Valid
      _))
  (if (Utilities|Time|IsValidTimerHandle (Variables|Default|GetTH_DelayFallDamageRemoval))
    (Development|PrintString "Ending Ability While Delayed Remove Active")
    (CallFunction|RemoveFallDamageImmunity)))

(fn BreakNearbyStructures
  (params (PlayerPawn :type Fort%20Player%20Pawn%20Object%20Reference))
  (CallFunction|BeginAsyncBreakNearbyStructuresOverlap (Math|Vector|MakeVector (.x (Transformation|GetActorLocation PlayerPawn)) (.y (Transformation|GetActorLocation PlayerPawn)) (+ (.z (Transformation|GetActorLocation PlayerPawn)) (Variables|Default|GetSmashBoxPosAdditionalZ))) (Variables|Default|GetSmash%20Box%20Extent) (Utilities|Array|MakeArray "ObjectTypeQuery1" "ObjectTypeQuery3") 0 (Utilities|Array|MakeArray PlayerPawn)) ; for performance reasons, we're moving this to c++ and it will\nbe handled async (ie we get a callback when it is done) 
)

(fn CheckForQuestProgress
  (bind _zipline (Variables|Default|GetZipline))
  (bind _returnvalue (GameplayTags|GetOwnedGameplayTags _zipline :!class /Script/GameplayTags.GameplayTagAssetInterface))
  (bind _playerpawn (Variables|Default|GetPlayerPawn))
  (if (Networking|HasAuthority _playerpawn)
    (bind _asfort_player_controller (Utilities|Casting|CastToFortPlayerController (Pawn|GetController _playerpawn)))
    (Quest|SendComplexCustomStatEvent (Quests|GetQuestManager _asfort_player_controller "Invalid") 0 "(GameplayTags=((TagName=\"Gameplay.Action.Player.Ziplining\")))" _returnvalue 1)
    (bind _result (CustomVerb|AddTagsforSubject (CustomVerb|MakeCustomVerbMessage _asfort_player_controller "(TagName=\"CustomVerb.Ascender.Use\")") "Zipline" _returnvalue))
    (CustomVerb|BroadcastCustomVerbMessage _result)
    (Ability|SendGameplayEventtoActor _asfort_player_controller "(TagName=\"CustomVerb.Ascender.Use\")")))

(fn CheckMotor
  (LocalVariables|SetHasHarness false)
  (bind _array_element (Utilities|Array|ForEachLoopwithBreak (Actor|GetAttachedActors (Variables|Default|GetPlayerPawn) true false))
    (:LoopBody
      (Utilities|Casting|CastToBP_ZipLine_Athena_Harness _array_element)
      (LocalVariables|SetHasHarness true))
    (:Completed
      _))
  (if (LocalVariables|GetHasHarness)
    _))

(fn GetZiplineMotorOffset
  (bind _asbp_athena_environmental_zip_line_spline (Utilities|Casting|CastToBP_Athena_Environmental_ZipLine_Spline (Variables|Default|GetZipline))
    (:then
      (return (Variables|/Game/Athena/Items/Consumables/ZipLine/SplineZipline/BP_Athena_Environmental_ZipLine_Spline.BP_Athena_Environmental_ZipLine_Spline_C|GetMotorOffset _asbp_athena_environmental_zip_line_spline)))
    (:CastFailed
      (return "0, 0, 0"))))

(fn GetZiplinePoleASocketLoc
  (bind _zipline (Variables|Default|GetZipline))
  (bind _asbp_athena_environmental_zip_line_spline (Utilities|Casting|CastToBP_Athena_Environmental_ZipLine_Spline _zipline)
    (:then
      (return (Variables|/Script/FortniteGame.FortAthenaZipline|GetEndPosition _asbp_athena_environmental_zip_line_spline)))
    (:CastFailed
      (return (Transformation|GetActorLocation _zipline)))))

(fn GetZiplinePoleBSocketLoc
  (bind _zipline (Variables|Default|GetZipline))
  (bind _asbp_athena_environmental_zip_line_spline (Utilities|Casting|CastToBP_Athena_Environmental_ZipLine_Spline _zipline)
    (:then
      (return (Variables|/Script/FortniteGame.FortAthenaZipline|GetStartPosition _asbp_athena_environmental_zip_line_spline)))
    (:CastFailed
      (return (Transformation|GetActorLocation _zipline)))))

(fn GetZiplineRotationForMotor
  (bind _zipline (Variables|Default|GetZipline))
  (bind _asbp_athena_environmental_zip_line_spline (Utilities|Casting|CastToBP_Athena_Environmental_ZipLine_Spline _zipline)
    (:then
      (return (Math|Rotator|MakeRotator (+ (.roll (Math|Rotator|MakeRotfromZ (- (Variables|/Script/FortniteGame.FortAthenaZipline|GetEndPosition _asbp_athena_environmental_zip_line_spline) (Variables|/Script/FortniteGame.FortAthenaZipline|GetStartPosition _asbp_athena_environmental_zip_line_spline)))) 1.8e+02) (.pitch (Math|Rotator|MakeRotfromZ (- (Variables|/Script/FortniteGame.FortAthenaZipline|GetEndPosition _asbp_athena_environmental_zip_line_spline) (Variables|/Script/FortniteGame.FortAthenaZipline|GetStartPosition _asbp_athena_environmental_zip_line_spline)))) (.yaw (Math|Rotator|MakeRotfromZ (- (Variables|/Script/FortniteGame.FortAthenaZipline|GetEndPosition _asbp_athena_environmental_zip_line_spline) (Variables|/Script/FortniteGame.FortAthenaZipline|GetStartPosition _asbp_athena_environmental_zip_line_spline)))))))
    (:CastFailed
      (return (Transformation|GetActorRotation _zipline)))))

(fn SpawnMotor
  (params (PlayerPawn :type Actor%20Object%20Reference))
  ;[0 0 0] Spawn motor and rotate components
  (Utilities|IsValid (Variables|Default|GetZiplineHarness)
    (:Is%20Not%20Valid
      (if (Networking|HasAuthority PlayerPawn)
        (bind _returnvalue (Game|SpawnActorfromClass "/Game/Gadgets/Zipline/Meshes/BP_ZipLine_Athena_Harness_Yellow.BP_ZipLine_Athena_Harness_Yellow_C" (Math|Transform|MakeTransform (Transformation|GetActorLocation (Variables|Default|GetPlayerPawn))) "AlwaysSpawn" "OverrideRootScale" 0 (Variables|Default|GetPlayerPawn) (Variables|Default|GetZipline) 0))
        (Variables|Default|SetZiplineHarness _returnvalue)
        (bind _offset (CallFunction|GetZiplineMotorOffset))
        (Class|BPZipLineAthenaHarness|OrientMotorforZipline (Variables|Default|GetZiplineHarness) (Variables|Default|GetZipline) _offset)
        (EventDispatchers|BindEventtoZiplineStateChanged (Variables|Default|GetPlayerPawn) (EventDispatchers|CreateEvent :!event HandleZiplineStateChanged))))))
