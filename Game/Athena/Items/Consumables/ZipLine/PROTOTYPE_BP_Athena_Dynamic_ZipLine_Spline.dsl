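; Prototype zipline actor that strings a rideable spline cable between two pole sockets.
; Construction shapes the spline from interpolated vertical/horizontal vectors with optional straightening, builds spline meshes with dynamic materials, and snaps the end point next tick.
; BeginPlay initializes the zipline and sends temporary auto-attach events to any instigator pawn, bridging to the external deployer plugin.
; Riding sends enter/exit gameplay events, tracks pawns and timestamps materials; death hides the actor.
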(event EventBeginPlay
  (bind _returnvalue (Game|GetInstigator))
  (bind _ziplinespline (Variables|FortAthenaSplineZipline|GetZiplineSpline))
  (bind _returnvalue_2 (Spline|GetNumberOfSplinePoints _ziplinespline))
  (if (> _returnvalue_2 0)
    (Zipline|Initialize (Spline|GetLocationatSplinePoint _ziplinespline 0 "World") (Spline|GetLocationatSplinePoint _ziplinespline (- _returnvalue_2 1) "World"))
    ;[0.8438 0.8264 0.04508 0.5] TODO: hotfix hook ups / clean up how we do the auto attach logic
    ; NOTE: This BP is outside the deployer plugin, so need to solve that first before we can hotfix
    ; NOTE: curently setting instigator as item using pawn or null to communicate between PRJ in plugin and this BP
    (Utilities|IsValid _returnvalue
      (:Is%20Valid
        (bind _asfort_player_pawn (Utilities|Casting|CastToFortPlayerPawn _returnvalue))
        (CallFunction|SendAutoAttachTEMPPROTO _asfort_player_pawn)))))

(event EventOnZipliningStarted
  (params (InteractingPawn :type Fort%20Player%20Pawn%20Object%20Reference) (InteractComponent :type Primitive%20Component%20Object%20Reference))
  ;[0.3209 0.3489 1] VFX for when local player is attached/detached from zipLine
  (|CallParentFunction InteractingPawn InteractComponent)
  (Utilities|FlowControl|Sequence
    (:then_0
      ;;[0.8958 0.5027 0 0.5] TEMP for triggering GA_Athena_ZipLine_Smash_Copy_PROTO
      (CallFunction|SendEnterEventTEMPPROTO InteractingPawn))
    (:then_1
      (CallFunction|AddPlayertoZipliningArray InteractingPawn)
      (for _array_element (Variables|Default|GetSplineMaterials)
        (Rendering|Material|SetScalarParameterValue _array_element "TimeStamp" (Utilities|Time|GetGameTimeinSeconds) :!class /Script/Engine.MaterialInstanceDynamic)))))

(event EventOnZipliningStopped
  (params (InteractingPawn :type Fort%20Player%20Pawn%20Object%20Reference) (InteractComponent :type Primitive%20Component%20Object%20Reference) (ZiplineUsageDuration :type Float%20%28single-precision%29))
  ;[0.3209 0.3489 1] VFX for when local player is attached/detached from zipLine
  (|CallParentFunction InteractingPawn InteractComponent ZiplineUsageDuration)
  (Utilities|FlowControl|Sequence
    (:then_0
      ;;[0.8958 0.5027 0 0.5] TEMP for triggering GA_Athena_ZipLine_Smash_Copy_PROTO
      (CallFunction|SendExitEventTEMPPROTO InteractingPawn))
    (:then_1
      (CallFunction|RemovePlayerfromZipliningArray InteractingPawn)
      (for _array_element (Variables|Default|GetSplineMaterials)
        (Rendering|Material|SetScalarParameterValue _array_element "TimeStamp" (Utilities|Time|GetGameTimeinSeconds) :!class /Script/Engine.MaterialInstanceDynamic)))))

(event Custom|WaitThenSetDynamicSpline
  (bind _splinemeshcomponents (Variables|Default|GetSplineMeshComponents))
  (bind _returnvalue (Utilities|Array|LastIndex _splinemeshcomponents))
  (bind _output (Utilities|Array|Get(acopy) _splinemeshcomponents _returnvalue))
  (bind _polebsocketlocation (Variables|Default|GetPoleBSocketLocation))
  ;[1 1 1] This currently is what allows us to set the spline points after construction. Trying to do so on the same frame does not work.
  (Utilities|FlowControl|DelayUntilNextTick)
  (Utilities|IsValid _output
    (:Is%20Valid
      (SplineMesh|SetEndPosition _output (Math|Transform|TransformLocation (Math|Transform|InvertTransform (Transformation|GetActorTransform)) _polebsocketlocation) true)
      (Spline|SetLocationatSplinePoint (Variables|FortAthenaSplineZipline|GetZiplineSpline) (- (Spline|GetNumberOfSplinePoints (Variables|FortAthenaSplineZipline|GetZiplineSpline)) 1) _polebsocketlocation "World" true) ; start point
    )))

(event EventOnDeathServer
  (params (Damage :type Float%20%28single-precision%29) (DamageTags :type Gameplay%20Tag%20Container%20Structure%20%28by%20ref%29) (Momentum :type Vector :default "0, 0, 0") (HitInfo :type Hit%20Result%20Structure%20%28by%20ref%29) (InstigatedBy :type Controller%20Object%20Reference) (DamageCauser :type Actor%20Object%20Reference) (EffectContext :type Gameplay%20Effect%20Context%20Handle%20Structure))
  ;[1 1 1] Hide spline & eject anyone currently ziplining
  (Rendering|SetActorHiddenInGame true))

(fn AddPlayerToZipliningArray
  (params (NewItem :type Fort%20Player%20Pawn%20Object%20Reference%20%28by%20ref%29))
  (Utilities|Array|AddUnique (Variables|Default|GetPlayerPawnArray) NewItem)
  (if (Variables|Default|GetDebug)
    (Development|PrintString (Utilities|String|Append "Zipline Player Array Size:" (Utilities|String|ToString(Integer) (Utilities|Array|Length (Variables|Default|GetPlayerPawnArray)))) true true "(R=0.045751,G=1.000000,B=0.143074,A=1.000000)" 5.0)
    (return)))

(fn AddSplineMeshSegment
  (bind _returnvalue (AddComponent|NODE_AddSplineMeshComponent-2))
  (Components|StaticMesh|SetStaticMesh _returnvalue (Variables|Default|GetSplineStaticMesh))
  (SplineMesh|SetForwardAxis _returnvalue (Variables|Default|GetForwardMeshAxis) true)
  (Collision|SetCollisionProfileName _returnvalue "FortTriggerOnlyInteractions" false :!class /Script/Engine.PrimitiveComponent)
  (Variables|/Script/Engine.PrimitiveComponent|SetbGenerateOverlapEvents true _returnvalue)
  (Utilities|Array|Add (Variables|Default|GetSplineMeshComponents) _returnvalue)
  (return _returnvalue))

(fn AutoSmoothTanget
  (params (Tangent :type Vector) (PointA :type Vector) (PointB :type Vector))
  (return (select (Variables|TangentSmoothing|GetAutoSmoothTangents) (* (* (Math|Vector|Normalize Tangent) (Math|Vector|VectorLength (- PointA PointB))) (Variables|TangentSmoothing|GetTangentSmoothStrength)) Tangent)))

(fn BlueprintCanInteract
  (params (InteractingPawn :type Fort%20Pawn%20Object%20Reference) (InteractionBeingAttempted :type EInteractionBeingAttempted%20Enum :default "FirstInteraction") (InteractionType :type TInteractionType%20Enum :default "IT_NoInteraction"))
  (return (not (GameplayTags|HasAnyMatchingGameplayTags InteractingPawn (Variables|Default|GetBlockInteractTags)))))

(fn CalcAutoLocationAtAlpha
  (params (InAlpha :type Float%20%28double-precision%29) (DrawDebug :type Boolean))
  (bind _lowerendlocation (Variables|Default|GetLowerEndLocation))
  (bind _higherendlocation (Variables|Default|GetHigherEndLocation))
  (bind _returnvalue (- _higherendlocation _lowerendlocation))
  (bind _inalpha (Variables|GetInAlpha))
  (bind _inalpha_2 (Variables|GetInAlpha))
  (bind _lowerendlocation_2 (Variables|Default|GetLowerEndLocation))
  (bind _higherendlocation_2 (Variables|Default|GetHigherEndLocation))
  (bind _lowerendlocation_3 (Variables|Default|GetLowerEndLocation))
  (bind _inalpha_3 (Variables|GetInAlpha))
  ;[1 1 1] Calc Auto Location for alpha/factor along spline
  (bind (_vertvec _horizvec) (CallFunction|GetAutoHorizandVertVectors (Variables|Default|GetHigherEndLocation) (Variables|Default|GetLowerEndLocation)))
  (if DrawDebug
    ;[1 1 1] DebugDisplay
    (Rendering|Debug|DrawDebugLine (Variables|Default|GetHigherEndLocation) (+ (Variables|Default|GetLowerEndLocation) (- (- _higherendlocation_2 (* _vertvec -1.0)) _lowerendlocation_3)) "(R=1.000000,G=0.921126,B=0.712643,A=1.000000)" 5.0 4.0)
    (Rendering|Debug|DrawDebugLine (Variables|Default|GetLowerEndLocation) (+ (Variables|Default|GetLowerEndLocation) (* _horizvec -1.0)) "(R=0.685079,G=0.707206,B=1.000000,A=1.000000)" 5.0 4.0)
    (Rendering|Debug|DrawDebugLine (Math|Vector|Lerp(Vector) (Math|Vector|Lerp(Vector) "0, 0, 0" (* _horizvec -1.0) _inalpha) (Math|Vector|Lerp(Vector) (- (- _higherendlocation_2 (* _vertvec -1.0)) _lowerendlocation_3) _returnvalue _inalpha_3) _inalpha_2) (+ (Math|Vector|Lerp(Vector) (Math|Vector|Lerp(Vector) "0, 0, 0" (* _horizvec -1.0) _inalpha) (Math|Vector|Lerp(Vector) (- (- _higherendlocation_2 (* _vertvec -1.0)) _lowerendlocation_3) _returnvalue _inalpha_3) _inalpha_2) "0, 0,500.000000") "(R=1.000000,G=0.000000,B=0.000000,A=1.000000)" 5.0 4.0))
  (if (or (> (Variables|Default|GetAutoLinearFactorHigh) 0.0) (> (Variables|Default|GetAutoLinearFactorLow) 0.0))
    ;[1 1 1] Lerp to the linear version of the point
    (return (Math|Vector|Lerp(Vector) (+ _lowerendlocation_2 (Math|Vector|Lerp(Vector) (Math|Vector|Lerp(Vector) "0, 0, 0" (* _horizvec -1.0) _inalpha) (Math|Vector|Lerp(Vector) (- (- _higherendlocation_2 (* _vertvec -1.0)) _lowerendlocation_3) _returnvalue _inalpha_3) _inalpha_2)) (Math|Vector|Lerp(Vector) (Variables|Default|GetLowerEndLocation) (Variables|Default|GetHigherEndLocation) (Variables|GetInAlpha)) (Math|Float|Clamp(Float) (Math|Float|Lerp (Variables|Default|GetAutoLinearFactorHigh) (Variables|Default|GetAutoLinearFactorLow) (Variables|GetInAlpha)))))
    (else
      (return (+ _lowerendlocation_2 (Math|Vector|Lerp(Vector) (Math|Vector|Lerp(Vector) "0, 0, 0" (* _horizvec -1.0) _inalpha) (Math|Vector|Lerp(Vector) (- (- _higherendlocation_2 (* _vertvec -1.0)) _lowerendlocation_3) _returnvalue _inalpha_3) _inalpha_2))))))

(fn CalculatePositionOfPoles
  (bind _polea (Variables|Default|GetPoleA))
  (bind _poleb (Variables|Default|GetPoleB))
  (Utilities|IsValid _polea
    (:Is%20Valid
      (Variables|Default|SetPoleASocketLocation (Transformation|GetSocketLocation (Variables|/Script/Engine.Actor|GetRootComponent _polea) "LineAttachSocket"))
      (Utilities|IsValid _poleb
        (:Is%20Valid
          (Variables|Default|SetPoleBSocketLocation (Transformation|GetSocketLocation (Variables|/Script/Engine.Actor|GetRootComponent _poleb) "LineAttachSocket")))
        (:Is%20Not%20Valid
          (Variables|Default|SetPoleBSocketLocation (Spline|GetLocationandTangentatSplinePoint (Variables|FortAthenaSplineZipline|GetZiplineSpline) (- (Spline|GetNumberOfSplinePoints (Variables|FortAthenaSplineZipline|GetZiplineSpline)) 1) "Local")))))
    (:Is%20Not%20Valid
      (Variables|Default|SetPoleASocketLocation (Spline|GetLocationandTangentatSplinePoint (Variables|FortAthenaSplineZipline|GetZiplineSpline) 0 "Local")))))

(fn GetAutoHorizAndVertVectors
  (params (highVector :type Vector) (LowVector :type Vector))
  (bind (_returnvalue_x _returnvalue_y _returnvalue_z) (Math|Vector|BreakVector (- highVector LowVector)))
  (return (Math|Vector|MakeVector 0.0 0.0 (* _returnvalue_z (Variables|Default|GetAutoSplineTangentVertCoef))) (* (Math|Vector|MakeVector _returnvalue_x _returnvalue_y) (Variables|Default|GetAutoSplineTangentHorizCoef))))

(fn RemovePlayerFromZipliningArray
  (params (Item :type Fort%20Player%20Pawn%20Object%20Reference%20%28by%20ref%29))
  (Utilities|Array|RemoveItem (Variables|Default|GetPlayerPawnArray) Item)
  (if (Variables|Default|GetDebug)
    (Development|PrintString (Utilities|String|Append "Zipline Player Array Size:" (Utilities|String|ToString(Integer) (Utilities|Array|Length (Variables|Default|GetPlayerPawnArray)))) true true "(R=1.000000,G=0.859266,B=0.014152,A=1.000000)" 5.0)
    (return)))

(fn SendAutoAttach_TEMP_PROTO
  (params (Actor :type Actor%20Object%20Reference))
  (Ability|SendGameplayEventtoActor Actor "(TagName=\"Ability.ZiplineDeployer.Zipline.AutoAttach\")" (Utilities|Struct|MakeGameplayEventData "(TagName=\"\")" 0 0 self)))

(fn SendEnterEvent_TEMP_PROTO
  (params (Actor :type Actor%20Object%20Reference))
  (Ability|SendGameplayEventtoActor Actor "(TagName=\"Ability.ZiplineDeployer.Zipline.Enter\")" (Utilities|Struct|MakeGameplayEventData "(TagName=\"\")" 0 0 self)))

(fn SendExitEvent_TEMP_PROTO
  (params (Actor :type Actor%20Object%20Reference))
  (Ability|SendGameplayEventtoActor Actor "(TagName=\"Ability.ZiplineDeployer.Zipline.Exit\")" (Utilities|Struct|MakeGameplayEventData "(TagName=\"\")" 0 0 self)))

(fn SetCableTilingBySplineLength
  (params (Target :type Mesh%20Component%20Object%20Reference))
  (Rendering|Material|SetScalarParameterValueonMaterials Target "Tile_V" (Math|Conversions|ToFloat(Integer) (Math|Float|Truncate (/ (Spline|GetSplineLength (Variables|FortAthenaSplineZipline|GetZiplineSpline)) (Variables|Default|GetTilingDivisor))))))

(fn SetSplinePositionAndTangent
  (params (SetPosition :type Boolean) (SetTangent :type Boolean) (ID :type Integer))
  (bind _splinemeshcomponents (Variables|Default|GetSplineMeshComponents))
  (bind _returnvalue (Utilities|Array|LastIndex _splinemeshcomponents))
  (bind _ziplinespline (Variables|FortAthenaSplineZipline|GetZiplineSpline))
  (bind _returnvalue_2 (Spline|GetNumberOfSplinePoints _ziplinespline))
  (bind _returnvalue_3 (- _returnvalue_2 1))
  (bind _returnvalue_4 (Math|Conversions|ToFloat(Integer) _returnvalue_3))
  (bind _output (Utilities|Array|Get(acopy) _splinemeshcomponents _returnvalue))
  (bind _alphaforlocation (LocalVariables|GetAlphaForLocation))
  (bind _onedivnumpoints (LocalVariables|GetOneDivNumPoints))
  (bind _onedivnumpoints_2 (LocalVariables|GetOneDivNumPoints))
  (bind _autosplinetangentlengthcoef (Variables|Default|GetAutoSplineTangentLengthCoef))
  ;[1 1 1] Set up alphas
  (bind _output_get (LocalVariables|SetSplinePointID ID))
  (LocalVariables|SetSplinePointIDToSet _output_get)
  (if (== (LocalVariables|GetSplinePointID) 0)
    (LocalVariables|SetSplinePointID 1)
    (LocalVariables|SetIsFirstPoint true))
  (if (and (> (LocalVariables|GetSplinePointID) 0) (> _returnvalue_2 1))
    (LocalVariables|SetOneDivNumPoints (/ 1.0 _returnvalue_4))
    (LocalVariables|SetAlpha (/ (Math|Conversions|ToFloat(Integer) (LocalVariables|GetSplinePointID)) _returnvalue_4))
    (LocalVariables|SetPrevAlpha (/ (Math|Conversions|ToFloat(Integer) (- (LocalVariables|GetSplinePointID) 1)) _returnvalue_4))
    (if (LocalVariables|GetIsFirstPoint)
      (LocalVariables|SetAlphaForLocation (LocalVariables|GetPrevAlpha))
      (else
        (LocalVariables|SetAlphaForLocation (LocalVariables|GetAlpha))))
    ;[1 1 1] Calc Auto Location for point
    (bind _pointlocation (CallFunction|CalcAutoLocationatAlpha (LocalVariables|GetAlphaForLocation)))
    (if (Variables|GetSetPosition)
      (Utilities|IsValid _output
        (:Is%20Valid
          (SplineMesh|SetEndPosition _output (Transformation|GetActorLocation) true)
          (Spline|SetLocationatSplinePoint (Variables|FortAthenaSplineZipline|GetZiplineSpline) (LocalVariables|GetSplinePointIDToSet) _pointlocation "World" false)))
      (else
        _))
    ;[1 1 1] Calc Auto Tangent for point
    (if (Variables|GetSetTangent)
      (bind (_vertvec _horizvec) (CallFunction|GetAutoHorizandVertVectors (Variables|Default|GetHigherEndLocation) (Variables|Default|GetLowerEndLocation)))
      (if (or (> (Variables|Default|GetAutoLinearFactorHigh) 0.0) (> (Variables|Default|GetAutoLinearFactorLow) 0.0))
        ;[1 1 1] Lerp to linear version
        (Spline|SetTangentatSplinePoint (Variables|FortAthenaSplineZipline|GetZiplineSpline) (LocalVariables|GetSplinePointIDToSet) (Math|Vector|Lerp(Vector) (* (Math|Vector|Lerp(Vector) (* _vertvec _onedivnumpoints) (* _horizvec _onedivnumpoints_2) _alphaforlocation) _autosplinetangentlengthcoef) (* (- (Variables|Default|GetHigherEndLocation) (Variables|Default|GetLowerEndLocation)) (LocalVariables|GetOneDivNumPoints)) (Math|Float|Clamp(Float) (Math|Float|Lerp (Variables|Default|GetAutoLinearFactorHigh) (Variables|Default|GetAutoLinearFactorLow) (LocalVariables|GetAlphaForLocation)))) "World" false)
        (else
          (Spline|SetTangentatSplinePoint (Variables|FortAthenaSplineZipline|GetZiplineSpline) (LocalVariables|GetSplinePointIDToSet) (* (Math|Vector|Lerp(Vector) (* _vertvec _onedivnumpoints) (* _horizvec _onedivnumpoints_2) _alphaforlocation) _autosplinetangentlengthcoef) "World" false))))))

(fn ConstructionScript
  (bind _ziplinespline (Variables|FortAthenaSplineZipline|GetZiplineSpline))
  (bind _returnvalue (Spline|GetNumberOfSplinePoints _ziplinespline))
  (bind _returnvalue_2 (- _returnvalue 1))
  (bind _ziplinespline_2 (Variables|FortAthenaSplineZipline|GetZiplineSpline))
  (bind _returnvalue_3 (Spline|GetNumberOfSplinePoints _ziplinespline_2))
  (bind _returnvalue_4 (- _returnvalue_3 2))
  (bind (_poleasocketlocation_x _poleasocketlocation_y _poleasocketlocation_z) (Math|Vector|BreakVector (Variables|GetPoleASocketLocation_X)))
  (bind (_polebsocketlocation_x _polebsocketlocation_y _polebsocketlocation_z) (Math|Vector|BreakVector (Variables|GetPoleBSocketLocation_X)))
  (bind _polea (Variables|Default|GetPoleA))
  (bind _poleb (Variables|Default|GetPoleB))
  (Utilities|IsValid _polea
    (:Is%20Valid
      (Variables|Default|SetPoleASocketLocation (Transformation|GetSocketLocation (Variables|/Script/Engine.Actor|GetRootComponent _polea) "LineAttachSocket"))
      (Utilities|IsValid _poleb
        (:Is%20Valid
          (Variables|Default|SetPoleBSocketLocation (Transformation|GetSocketLocation (Variables|/Script/Engine.Actor|GetRootComponent _poleb) "LineAttachSocket"))
          (if (< _poleasocketlocation_z _polebsocketlocation_z)
            (Variables|Default|SetLowerPointID 0)
            (Variables|Default|SetHigherPointID (- (Spline|GetNumberOfSplinePoints (Variables|FortAthenaSplineZipline|GetZiplineSpline)) 1))
            (Variables|Default|SetHigherEndLocation (Variables|Default|GetPoleASocketLocation))
            (Variables|Default|SetLowerEndLocation (Variables|Default|GetPoleBSocketLocation))
            (else
              (Variables|Default|SetHigherPointID 0)
              (Variables|Default|SetLowerPointID (- (Spline|GetNumberOfSplinePoints (Variables|FortAthenaSplineZipline|GetZiplineSpline)) 1))
              (Variables|Default|SetLowerEndLocation (Variables|Default|GetPoleASocketLocation))
              (Variables|Default|SetHigherEndLocation (Variables|Default|GetPoleBSocketLocation))))
          (for _index (range (+ _returnvalue_2 1))
            (if (or (== _index 0) (== _index _returnvalue_2))
              (CallFunction|SetSplinePositionandTangent (Variables|Default|GetAuto%20Set%20Spline%20Ends) (Variables|Default|GetAuto%20Set%20Spline%20Ends) _index)
              (else
                (CallFunction|SetSplinePositionandTangent (Variables|Default|GetAuto%20Set%20Spline%20Mids) (Variables|Default|GetAuto%20Set%20Spline%20Mids) _index))))
          (Spline|UpdateSpline (Variables|FortAthenaSplineZipline|GetZiplineSpline)))))
    (:Is%20Not%20Valid
      _))
  (for _index_2 (range (+ _returnvalue_4 1))
    (bind _splinemeshsegment (CallFunction|AddSplineMeshSegment))
    (SplineMesh|SetStartandEnd _splinemeshsegment (Spline|GetLocationandTangentatSplinePoint (Variables|FortAthenaSplineZipline|GetZiplineSpline) _index_2 "Local") (CallFunction|AutoSmoothTanget (.Tangent (Spline|GetLocationandTangentatSplinePoint (Variables|FortAthenaSplineZipline|GetZiplineSpline) _index_2 "Local")) (Spline|GetLocationandTangentatSplinePoint (Variables|FortAthenaSplineZipline|GetZiplineSpline) _index_2 "Local") (Spline|GetLocationandTangentatSplinePoint (Variables|FortAthenaSplineZipline|GetZiplineSpline) (+ _index_2 1) "Local")) (Spline|GetLocationandTangentatSplinePoint (Variables|FortAthenaSplineZipline|GetZiplineSpline) (+ _index_2 1) "Local") (CallFunction|AutoSmoothTanget (.Tangent (Spline|GetLocationandTangentatSplinePoint (Variables|FortAthenaSplineZipline|GetZiplineSpline) (+ _index_2 1) "Local")) (Spline|GetLocationandTangentatSplinePoint (Variables|FortAthenaSplineZipline|GetZiplineSpline) _index_2 "Local") (Spline|GetLocationandTangentatSplinePoint (Variables|FortAthenaSplineZipline|GetZiplineSpline) (+ _index_2 1) "Local")) true :!class /Script/Engine.SplineMeshComponent)
    ;[1 1 1] Set Material Instances for allsplines Start and End Falloff
    (for _index_3 (range (Rendering|Material|GetNumMaterials _splinemeshsegment))
      (bind _returnvalue_5 (Rendering|Material|CreateDynamicMaterialInstance _splinemeshsegment _index_3 0 "None" :!class /Script/Engine.PrimitiveComponent))
      (Rendering|Material|SetScalarParameterValue _returnvalue_5 "UseStart" (Math|Float|SelectFloat 1.0 0.0 (== _index_2 _returnvalue_4)) :!class /Script/Engine.MaterialInstanceDynamic)
      (Rendering|Material|SetScalarParameterValue _returnvalue_5 "UseEnd" (Math|Float|SelectFloat 1.0 0.0 (== _index_2 0)) :!class /Script/Engine.MaterialInstanceDynamic)
      (Utilities|Array|Add (Variables|Default|GetSplineMaterials) _returnvalue_5)
      (CallFunction|SetCableTilingbySplineLength _splinemeshsegment)
      (CallFunction|WaitThenSetDynamicSpline))))
